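# Unlit translucent material for Niagara mesh particles, apparently the idle energy effect on the PinkSpike pickaxe.
# A local-Z axis-aligned fresnel, shaped by dynamic Power and vertical UV fades, forms the Base mask.
# A rotatable gradient, panned by dynamic Sweep and scaled by Sweep Fill, forms the Pulsing mask.
# A product of rerouted masks is tinted by particle color through emissive adjustments and multiplied by particle alpha.
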Material /BRCosmetics/Weapons/FORT_Melee/Pickaxe_PinkSpike/FX/M_PinkSpike_Energy_Idle
graph {
 properties {
  bUsedWithNiagaraMeshParticles=true
  blendMode=BLEND_Translucent
  pixelDepthOffsetMode=PDOM_Legacy
  previewMesh=&/PinkSpike_Axe_Sheen
  shadingModel=MSM_Unlit
  usedShadingModels=Unlit
 }
 nodes [
  n1 = ParticleColor()
  n2 = Multiply(A=NamedRerouteUsage(), B=NamedRerouteUsage())
  MaterialFunctionCall(Emissive=Multiply(A=@n2, B=@n1.RGB)) {materialFunction=&MF_FX_EmissiveAdjustments}
  Multiply(A=@n2, B=@n1.A)
  n3 = Add(A=MaterialFunctionCall(UVs=TextureCoordinate(), "Rotation Angle (0-1)"=ScalarParameter() {parameterName=Gradient Angle}) {materialFunction=&CustomRotator}.Rotated Values) {constB=0}
  n4 = Clamp(None=@n3)
  n5 = ScalarParameter() {defaultValue=1, parameterName=UV Clamp?}
  n6 = DynamicParameter() {paramNames [Power, Sweep, Sweep Fill, Param4]}
  n7 = TextureSampleParameter2D(UVs=Panner(Coordinate=LinearInterpolate(A=@n3, B=@n4, Alpha=@n5), Time=@n6.Sweep) {speedY=0.5}) {
   parameterName=Texture Gradient
   samplerSource=SSM_Clamp_WorldGroupSettings
   samplerType=SAMPLERTYPE_LinearColor
   texture=/Game/Global/Textures/Default/Masks_and_Gradients/T_VertGradient
  }
  NamedRerouteDeclaration(None=Clamp(None=Multiply(A=@n7.R, B=@n6.Sweep Fill) {constB=2}, Max=ScalarParameter() {defaultValue=1, parameterName=Gradient Clamp Max}) {maxDefault=0.5}) {name=Pulsing, nodeColor=<rgba 0 1 0.2031 1>}
  n8 = Transform(None=Constant3Vector() {constant=<rgba 0 0 1 0>}) {transformSourceType=TRANSFORMSOURCE_Local}
  n9 = OneMinus(None=MaterialFunctionCall(Float2=TextureCoordinate()) {materialFunction=&BreakOutFloat2Components}.G)
  n10 = LinearInterpolate(A=Constant() {r=1}, B=Constant() {r=1.2}, Alpha=Saturate(None=Multiply(A=@n9) {constB=4}))
  n11 = Power(
   Base=Saturate(None=MaterialFunctionCall(Axis=@n8, Exponent=@n10) {materialFunction=&AxisAlignedFresnel}.Fresnel)
   Exp=DynamicParameter() {paramNames [Power, Sweep, Sweep Fill, Param4]}.Power
  )
  n12 = Constant() {r=0.1}
  n13 = Add(A=@n12, B=Constant() {r=1})
  n14 = MaterialFunctionCall(Float2=TextureCoordinate()) {materialFunction=&BreakOutFloat2Components}
  n15 = LinearInterpolate(Alpha=Saturate(None=SmoothStep(Min=@n12, Max=@n13, Value=@n14.G))) {constA=0.5}
  NamedRerouteDeclaration(None=Saturate(None=Multiply(A=@n11, B=@n15))) {name=Base, nodeColor=<rgba 0 1 0.3672 1>}
  TextureSampleParameter2D() {parameterName=Texture Gradient, texture=&T_VertGradient}
 ]
}
