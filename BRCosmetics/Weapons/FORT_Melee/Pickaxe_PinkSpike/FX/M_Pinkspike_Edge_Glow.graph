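# Unlit additive particle material for the edge glow effect on the PinkSpike pickaxe, usable on Niagara sprites.
# It samples the red channel of a grayscale edge-glow texture, applies a power curve to tighten it, and tints it with particle color.
# The result feeds the emissive input of the shared MF_FX_EmissiveAdjustments function.
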Material /BRCosmetics/Weapons/FORT_Melee/Pickaxe_PinkSpike/FX/M_Pinkspike_Edge_Glow
graph {
 properties {bUsedWithNiagaraSprites=true, blendMode=BLEND_Additive, pixelDepthOffsetMode=PDOM_Legacy, shadingModel=MSM_Unlit, usedShadingModels=Unlit}
 nodes [
  n1 = TextureSample() {samplerType=SAMPLERTYPE_LinearGrayscale, texture=&/T_PinkSpike_Edge_Glow}
  MaterialFunctionCall(Emissive=Multiply(A=Power(Base=@n1.R) {constExponent=1.2}, B=ParticleColor().RGB)) {materialFunction=&MF_FX_EmissiveAdjustments}
 ]
}
